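# Niagara effect for the Fiend Hunter Crossbow reload, ejecting a mesh magazine through a single Magazine emitter.
# A burst spawns short-lived mesh particles at a local offset, launched in a cone with random speed and spin.
# Gravity and collision with random restitution make them fall and bounce while rotating.
# Curves fade alpha and shrink mesh scale late in life, and dynamic material parameters are written each update.
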NiagaraSystem /Game/Weapons/FORT_Crossbows/Crossbow/Effects/NS_FiendHunterCrossbow_Reload
dependencies {usedDataInterfaces [&NiagaraDataInterfaceCurve], usedDynamicInputs [], usedModules [&SystemState]}
emitters {Magazine {
  emitterUpdateScript [
   &EmitterState(Life Cycle Mode=::NewEnumerator1, Loop Behavior=::NewEnumerator1, Loop Duration=1, UseLoopDelay=false)
   &SpawnBurst_Instantaneous(Use Spawn Probability=false, UseLoopCountLimit=false)
  ]
  particleSpawnScript [
   &InitializeParticle(
    Lifetime=1.2
    Mesh Scale Mode=::NewEnumerator3
    Position Offset=<20 0 0>
    Position Offset Coordinate Space=::Local
    UsePositionOffset=true
    Write Material Random=true
   )
   &AddVelocityInCone(
    Cone Angle=60
    Module.Use Velocity Falloff On Cone Axis=false
    Velocity Distribution Along Cone Axis=1
    Velocity Strength=&RandomRangeFloat(Minimum=50, Maximum=100)
   )
   /SetVariables(Particles.SpawnRotationVector=&RandomRangeVector(Minimum=<0 1 0>, Maximum=<0 3 0>, Evaluation Type=::NewEnumerator1)) {isSetParameters=true}
  ]
  particleUpdateScript [
   &ParticleState()
   &ScaleColor(
    Scale Alpha=&FloatFromCurve(FloatCurve=DataInterface() {propertyValues {Curve {keys [<at 0.5 1.0113 RCIM_Cubic>,<at 1.0011 0.002 RCIM_Cubic>]}}})
    ScaleA=true
    ScaleRGB=false
   )
   &GravityForce()
   &Collision(
    Age Colliding Particles=false
    Module.Control Rest State Via Surface Normal=false
    Module.EnableMaxCollisionCount=false
    Module.Randomize Collision Normal Vector=false
    Restitution=&RandomRangeFloat(Minimum=0.2, Maximum=0.5)
   )
   &SolveForcesAndVelocity(Clamp Velocity=false, Limit Acceleration=false)
   &ScaleMeshSize(Scale Factor=&VectorFromFloat(Value=&FloatFromCurve(FloatCurve=DataInterface() {propertyValues {Curve {keys [<at 0.6 1 RCIM_Cubic>, {interpMode=RCIM_Cubic, leaveTangent=-3.8891, time=1, value=0}]}}}))) {module=ScaleMeshSize001}
   &DynamicMaterialParameters(
    Index 0 Param 2=1
    Index 0 Param 3=1
    Index 0 Param 4=1
    Module.Param0WriteEnabled=true
    Module.Param1WriteEnabled=true
    Module.Param2WriteEnabled=true
    Module.Param3WriteEnabled=true
   )
   &UpdateMeshOrientation(Rotation Rate=0.3, Rotation Vector=$Particles.SpawnRotationVector)
  ]
  rendererClasses [&NiagaraMeshRendererProperties]
  renderers [{rendererClass=&NiagaraMeshRendererProperties, rendererIndex=0}]
 }}
summary {
 systemName=NS_FiendHunterCrossbow_Reload
 emitters [{emitterName=Magazine, rendererClasses [&NiagaraMeshRendererProperties]}]
 userVariables []
}
