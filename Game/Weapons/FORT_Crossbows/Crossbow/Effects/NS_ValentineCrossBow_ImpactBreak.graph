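# Niagara impact effect for the Valentine crossbow, named for a breaking-heart burst when a bolt hits.
# Seven emitters make up the burst: a bolt, streaks, a flash flare, smoke, a heart center, and left and right heart-break halves.
# Only a SystemState module is listed, with no user variables; renderers, data interfaces and per-emitter logic are not shown.
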NiagaraSystem /Game/Weapons/FORT_Crossbows/Crossbow/Effects/NS_ValentineCrossBow_ImpactBreak
dependencies {usedDataInterfaces [], usedDynamicInputs [], usedModules [&SystemState]}
emitters {
 NE_Bolt {rendererClasses [], renderers []}
 NE_FlashFlare {rendererClasses [], renderers []}
 NE_HeartBreakLeft {rendererClasses [], renderers []}
 NE_HeartBreakRight {rendererClasses [], renderers []}
 NE_HeartCenter {rendererClasses [], renderers []}
 NE_Smokes {rendererClasses [], renderers []}
 NE_Streaks {rendererClasses [], renderers []}
}
summary {
 systemName=NS_ValentineCrossBow_ImpactBreak
 emitters [
  {emitterName=NE_Bolt, rendererClasses []}
  {emitterName=NE_Streaks, rendererClasses []}
  {emitterName=NE_FlashFlare, rendererClasses []}
  {emitterName=NE_Smokes, rendererClasses []}
  {emitterName=NE_HeartCenter, rendererClasses []}
  {emitterName=NE_HeartBreakLeft, rendererClasses []}
  {emitterName=NE_HeartBreakRight, rendererClasses []}
 ]
 userVariables []
}
