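# Niagara muzzle flash effect for the Valentine crossbow, built from glow, smoke, spark and heart sprite emitters.
# NE_HeartSprites001 bursts a few short-lived sprites in a narrow cone, slowed by drag while they spin, grow then shrink, and fade out.
# A camera-distance fade based on owner position scales their size.
# The older NE_HeartSprites emitter is disabled; glow, smoke and spark emitters list no modules or renderers here.
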NiagaraSystem /Game/Weapons/FORT_Crossbows/Crossbow/Effects/NS_ValentineCrossBow_MuzzleFlash
dependencies {
 usedDataInterfaces [&NiagaraDataInterfaceCurve, &NiagaraDataInterfaceCamera, &NiagaraDataInterfaceColorCurve]
 usedDynamicInputs [&MakeVectorFromLinearColorRGB]
 usedModules [&SystemState]
}
emitters {
 NE_Glow {rendererClasses [], renderers []}
 NE_HeartSprites {bEnabled=false, rendererClasses [], renderers []}
 NE_HeartSprites001 {
  emitterUpdateScript [
   &EmitterState(Life Cycle Mode=::NewEnumerator1, Loop Behavior=::NewEnumerator1, Loop Duration=1, UseLoopDelay=false)
   &SpawnBurst_Instantaneous(
    Spawn Count=&RandomRangeInt(Minimum=6, Maximum=7)
    Use Spawn Probability=false
    UseLoopCountLimit=false
   )
  ]
  particleSpawnScript [
   &InitializeParticle(
    Color=&RandomRangeLinearColor(Minimum=<rgba 0.8 1 1 1>, Maximum=<rgba 1 1 1 1>)
    Lifetime=&RandomRangeFloat(Minimum=0.3, Maximum=0.5)
    Sprite Rotation Angle=&RandomRangeFloat(Minimum=-0.1, Maximum=0.1)
    Sprite Rotation Mode=::NewEnumerator1
    Sprite Size Mode=::NewEnumerator3
    Uniform Sprite Size=&RandomRangeFloat(Minimum=10, Maximum=15)
    UsePositionOffset=false
    Write Material Random=true
   )
   &AddVelocity(
    Cone Angle=10
    Module.UseConeFalloff=false
    Module.UseVelDistribution=false
    Override Seed=false
    Velocity Mode=::NewEnumerator2
    Velocity Speed=&RandomRangeFloat(Minimum=200, Maximum=800)
   )
  ]
  particleUpdateScript [
   &ParticleState()
   &SpriteRotationRate(Rotation Rate=&RandomRangeFloat(Minimum=-180, Maximum=180))
   &ScaleSpriteSize(
    Uniform Curve Scale=/Niagara/DynamicInputs/Camera/CameraFade(
     Camera Query:NiagaraDataInterfaceCamera=DataInterface()
     Fade Offset:NiagaraFloat=150
     Fade Distance:NiagaraFloat=100
     Use Owner Position:NiagaraBool=true
    )
    Uniform Curve Sprite Scale=DataInterface() {propertyValues {Curve {keys [
        {arriveTangent=4.3331, interpMode=RCIM_Cubic, leaveTangent=4.3331, tangentMode=RCTM_User, time=0.05, value=0}
        {interpMode=RCIM_Cubic, tangentMode=RCTM_User, time=0.3, value=1}
        {interpMode=RCIM_Cubic, tangentMode=RCTM_User, time=1, value=0.3}
       ]}}}
   )
   &Drag(
    Drag=&RandomRangeFloat(Minimum=3, Maximum=7)
    Rotational Drag=0
    Use Linear Drag=true
    Use Rotational Drag=true
   )
   &SolveForcesAndVelocity(Clamp Velocity=false, Limit Acceleration=false)
   &ScaleColor(
    Scale Mode=1
    Scale RGBA=&MakeVector4_FromLinearColor(LinearColor=&ColorFromCurve(ColorCurve=DataInterface() {propertyValues {AlphaCurve {keys [<at 0.5461 1.5>,<at 1 0>]}, BlueCurve {keys [<at 0 1>]}, GreenCurve {keys [<at 0 1>]}, RedCurve {keys [<at 0 1>]}}}))
    ScaleRGBA=true
   )
  ]
  rendererClasses [&NiagaraSpriteRendererProperties]
  renderers [{rendererClass=&NiagaraSpriteRendererProperties}]
 }
 NE_Smokes {rendererClasses [], renderers []}
 NE_Sparks {rendererClasses [], renderers []}
}
summary {
 systemName=NS_ValentineCrossBow_MuzzleFlash
 emitters [
  {bEnabled=false, emitterName=NE_HeartSprites, rendererClasses []}
  {emitterName=NE_Glow, rendererClasses []}
  {emitterName=NE_Smokes, rendererClasses []}
  {emitterName=NE_Sparks, rendererClasses []}
  {emitterName=NE_HeartSprites001, rendererClasses [&NiagaraSpriteRendererProperties]}
 ]
 userVariables []
}
