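# Niagara hit effect for the Valentine crossbow striking a player: a pink heart pop with a burst of fast orange streaks.
# NE_Heart spawns two burst groups tagged by VisibilityTag, driving two mesh renderers, a light and a sprite through per-renderer attribute modules.
# Camera-oriented hearts use curves for scale overshoot, color fade and dynamic material parameters.
# NE_Streaks launches short-lived stretched sprites; NE_HeartSprites and NE_HeartDissolve list no renderers.
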NiagaraSystem /Game/Weapons/FORT_Crossbows/Crossbow/Effects/NS_ValentineCrossbow_PlayerHit
dependencies {
 usedDataInterfaces [
  &NiagaraDataInterfaceCamera
  &NiagaraDataInterfaceCurve
  &NiagaraDataInterfaceColorCurve
  &NiagaraDataInterfaceVector2DCurve
 ]
 usedDynamicInputs []
 usedModules [&SystemState]
}
emitters {
 NE_Heart {
  emitterUpdateScript [
   &EmitterState()
   &SpawnBurst_Instantaneous(Use Spawn Probability=false, UseLoopCountLimit=false) {module=SpawnBurst_Instantaneous001}
   &SpawnBurst_Instantaneous(Spawn Group=1, Use Spawn Probability=false, UseLoopCountLimit=false)
  ]
  particleSpawnScript [
   /SetVariables(Particles.VisibilityTag=&SpawnGroup()) {isSetParameters=true}
   /SetVariables(Particles.BoltCameraOffset=20) {isSetParameters=true}
   &InitializeParticle(Color=<rgba 1 0.1119 0.1714 1>, Lifetime=0.6, Mesh Scale Mode=::NewEnumerator4, UsePositionOffset=false, Write Material Random=true)
   /SetVariables(Particles.BoltOrientation:Quat4f=&OrientMeshToVector(
     Look At Direction=&ReturnCameraVector(Camera Vectors=::NewEnumerator1, Coordinate Space=::Simulation)
     Reference Axis Vector=<1 0 0>
    )) {isSetParameters=true}
  ]
  particleUpdateScript [
   &ParticleState()
   &ScaleMeshSize(Scale Factor=&VectorFromFloat(Value=&FloatFromCurve(FloatCurve=DataInterface() {propertyValues {Curve {keys [
          {arriveTangent=5.9342, interpMode=RCIM_Cubic, leaveTangent=5.9342, tangentMode=RCTM_User, time=0, value=0}
          {arriveTangent=2.5409, interpMode=RCIM_Cubic, leaveTangent=2.5409, time=0.25, value=1.1465}
          {arriveTangent=-0.8874, interpMode=RCIM_Cubic, leaveTangent=-0.8874, tangentMode=RCTM_User, time=0.3352, value=0.8517}
          {arriveTangent=-0.0035, interpMode=RCIM_Cubic, leaveTangent=-0.0035, tangentMode=RCTM_User, time=0.5314, value=0.7821}
          {arriveTangent=4.0398, interpMode=RCIM_Cubic, leaveTangent=4.0398, tangentMode=RCTM_User, time=1, value=2}
         ]}}})))
   &ScaleColor(
    Scale Alpha=&FloatFromCurve(FloatCurve=DataInterface() {propertyValues {Curve {keys [<at 0.5 1 RCIM_Cubic>,<at 1 0>]}}})
    Scale RGB=&VectorFromFloat(Value=&FloatFromCurve(FloatCurve=DataInterface() {propertyValues {Curve {keys [{interpMode=RCIM_Cubic, tangentMode=RCTM_User, time=0, value=5}, {interpMode=RCIM_Cubic, tangentMode=RCTM_User, time=0.6, value=1}]}}}))
    ScaleA=true
    ScaleRGB=true
   )
   &DynamicMaterialParameters(
    Index 0 Param 1=&FloatFromCurve(FloatCurve=DataInterface() {propertyValues {Curve {keys [<at 0.1 0 RCIM_Cubic>,<at 0.27 1.5 RCIM_Cubic>,<at 0.6 0>]}}})
    Index 0 Param 2=7
    Index 1 Param 1=&FloatFromCurve(FloatCurve=DataInterface() {propertyValues {Curve {keys [<at 0 1.25 RCIM_Cubic>, {arriveTangent=-1.25, interpMode=RCIM_Cubic, leaveTangent=-1.25, time=0.5, value=0.2247}, <at 1 0>]}}})
    Index 1 Param 3=1.25
    Index 2 Param 2=&FloatFromCurve(FloatCurve=DataInterface() {propertyValues {Curve {keys [<at 0 2 RCIM_Cubic>,<at 1 0>]}}})
    Index 2 Param 3=1
    Index 2 Param 4=1
    Module.Index0Float0Write=true
    Module.Index0Float1Write=true
    Module.Index0Float2Write=true
    Module.Index0Float3Write=true
    Module.Index1Float0Write=true
    Module.Index1Float1Write=true
    Module.Index1Float2Write=true
    Module.Index1Float3Write=true
    Module.Index2Float0Write=true
    Module.Index2Float1Write=true
    Module.Index2Float2Write=true
    Module.Index2Float3Write=true
    Write Parameter Index 1=::NewEnumerator1
    Write Parameter Index 2=::NewEnumerator1
   )
   &CameraOffset(Camera Offset Amount=15)
   &FN_Mesh_Attributes(
    Create Color Attribute=true
    Create Orientation Attribute=true
    Create Scale Attribute=true
    Orientation:Quat4f=&MakeQuaternion(Quaternion From=0, Axis=<0 0 1>, Angle=0.5, From=?, To=?, Angle Type=::NewEnumerator1, Normalize Vectors=false)
    Scale=&Multiply_VectorByFloat(
     Vector=<1 1.3 1.3>
     Float=&FloatFromCurve(FloatCurve=DataInterface() {propertyValues {Curve {keys [<at 0.8 1 RCIM_Cubic>,<at 1 0>]}}})
    )
    Write to Module Namespace=true
   )
   &FN_Light_Attributes(
    Color=&ColorFromCurve(ColorCurve=DataInterface() {propertyValues {
       AlphaCurve {keys [<at 0 1 RCIM_Cubic>]}
       BlueCurve {keys [<at 0 0.3906>,<at 0.1072 11.7187 RCIM_Cubic>,<at 0.9841 0.3906>]}
       GreenCurve {keys [<at 0 0.1198>,<at 0.1072 3.5938 RCIM_Cubic>,<at 0.9841 0.1198>]}
       RedCurve {keys [<at 0 1>,<at 0.1072 30 RCIM_Cubic>,<at 0.9841 1>]}
      }})
    Create Color Attribute=true
    Create Exponent Attribute=true
    Create Radius Attribute=true
    Exponent=1
    Radius=70
    Scale Color=2
   )
   &FN_Sprite_Attributes(
    Color=&MakeLinearColorFromVectorAndFloat(
     "Vector (RGB)"=&MakeVectorFromLinearColorRGB(Color=&ColorFromCurve(ColorCurve=DataInterface() {propertyValues {
         AlphaCurve {keys [<at 0 1 RCIM_Cubic>]}
         BlueCurve {keys [<at 0 1.0141 RCIM_Cubic>,<at 0.9806 0.949>]}
         GreenCurve {keys [<at 0 0.1388 RCIM_Cubic>,<at 0.9806 0.1707>]}
         RedCurve {keys [<at 0 2 RCIM_Cubic>,<at 0.9806 1>]}
        }}))
     "Float (Alpha)"=&FloatFromCurve(FloatCurve=DataInterface() {propertyValues {Curve {keys [<at 0 1 RCIM_Cubic>, {arriveTangent=-0.3208, leaveTangent=-0.3208, time=0.4176, value=0.3208}, <at 1 0>]}}})
    )
    Create Camera Offset Attribute=true
    Create Color Attribute=true
    Create SpriteSize Attribute=true
    SpriteSize=&Vector2DFromFloat(Value=&Multiply_Float(
      A=150
      B=&FloatFromCurve(FloatCurve=DataInterface() {propertyValues {Curve {keys [
           {arriveTangent=2.8071, interpMode=RCIM_Cubic, leaveTangent=2.8071, tangentMode=RCTM_User, time=0, value=0}
           {interpMode=RCIM_Cubic, tangentMode=RCTM_User, time=1, value=1}
          ]}}})
     ))
    Write to Module Namespace=true
   )
  ]
  rendererClasses [
   &NiagaraMeshRendererProperties
   &NiagaraLightRendererProperties
   &NiagaraSpriteRendererProperties
  ]
  renderers [
   {rendererClass=&NiagaraMeshRendererProperties, rendererIndex=0}
   {rendererClass=&NiagaraMeshRendererProperties, rendererIndex=1}
   {rendererClass=&NiagaraLightRendererProperties, rendererIndex=2}
   {rendererClass=&NiagaraSpriteRendererProperties, rendererIndex=3}
  ]
 }
 NE_HeartDissolve {rendererClasses [], renderers []}
 NE_HeartSprites {rendererClasses [], renderers []}
 NE_Streaks {
  emitterUpdateScript [
   &EmitterState()
   &SpawnBurst_Instantaneous(
    Spawn Count=&RandomRangeInt(Minimum=5, Maximum=7)
    Use Spawn Probability=false
    UseLoopCountLimit=false
   )
  ]
  particleSpawnScript [
   &InitializeParticle(
    Color=<rgba 10 2.998 0 1>
    Lifetime=&RandomRangeFloat(Minimum=0.08, Maximum=0.14, Evaluation Type=::NewEnumerator1)
    Sprite Size=&RandomRangeVector2D(Minimum=<1.5 35>, Maximum=<2.5 55>, Evaluation Type=::NewEnumerator1)
    Sprite Size Mode=::NewEnumerator4
    UsePositionOffset=false
    Write Material Random=true
   )
   &AddVelocity(Velocity=&RandomRangeVector(Minimum=<250 0 0>, Maximum=<600 0 0>, Evaluation Type=::NewEnumerator1))
   &ShapeLocation(
    Hemisphere Distribution=<0.5 1>
    MaskSpawnGroup=false
    Override Seed=false
    Rotation Mode=::NewEnumerator2
    Sphere Radius=3
    UseOwnerScale=true
    Yaw / Pitch / Roll=<0.75 0 0>
   )
   &AddVelocity(Velocity=$Particles.ShapeLocation.ShapeVector, Velocity Speed Scale=70) {module=AddVelocity001}
  ]
  particleUpdateScript [
   &ParticleState()
   &ScaleSpriteSize(
    Scale Factor=&Vector2DFromCurve(
     Vector2Curve=DataInterface() {propertyValues {
       XCurve {keys [
         {interpMode=RCIM_Cubic, tangentMode=RCTM_User, time=0, value=1}
         <at 0.1 1 RCIM_Cubic>
         <at 0.5 1 RCIM_Cubic>
         {interpMode=RCIM_Cubic, tangentMode=RCTM_User, time=1, value=1}
        ]}
       YCurve {keys [
         {arriveTangent=21.2968, interpMode=RCIM_Cubic, leaveTangent=21.2968, tangentMode=RCTM_User, time=0, value=0}
         {arriveTangent=1.9804, interpMode=RCIM_Cubic, leaveTangent=1.9804, time=0.1, value=0.9902}
         {arriveTangent=-1.1002, interpMode=RCIM_Cubic, leaveTangent=-1.1002, time=0.5, value=0.9902}
         {arriveTangent=-3.8038, interpMode=RCIM_Cubic, leaveTangent=-3.8038, tangentMode=RCTM_User, time=1, value=0}
        ]}
      }}
     ScaleCurve=<1 1>
    )
    Scale Sprite Size Mode=::NewEnumerator2
   )
   &SolveForcesAndVelocity(Clamp Velocity=false, Limit Acceleration=false)
  ]
  rendererClasses [&NiagaraSpriteRendererProperties]
  renderers [{rendererClass=&NiagaraSpriteRendererProperties}]
 }
}
summary {
 systemName=NS_ValentineCrossbow_PlayerHit
 emitters [
  {
   emitterName=NE_Heart
   rendererClasses [
    &NiagaraMeshRendererProperties
    &NiagaraLightRendererProperties
    &NiagaraSpriteRendererProperties
   ]
  }
  {emitterName=NE_Streaks, rendererClasses [&NiagaraSpriteRendererProperties]}
  {emitterName=NE_HeartSprites, rendererClasses []}
  {emitterName=NE_HeartDissolve, rendererClasses []}
 ]
 userVariables []
}
