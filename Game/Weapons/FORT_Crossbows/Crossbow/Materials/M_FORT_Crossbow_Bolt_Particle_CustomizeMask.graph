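# Lit material for crossbow bolt mesh particles, including Niagara, that adds a vertex wobble to the shared gun master shading.
# Dynamic parameters drive a sine-timed rotation about a randomly blended particle-space axis and offset pivot, scaled by Wobble.
# The offset is blended into the gun shader's material attributes as World Position Offset, then passed through the shared FX emissive adjustment function.
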Material /Game/Weapons/FORT_Crossbows/Crossbow/Materials/M_FORT_Crossbow_Bolt_Particle_CustomizeMask
graph {
 properties {
  bUseMaterialAttributes=true
  bUsedWithMeshParticles=true
  bUsedWithNiagaraMeshParticles=true
  pixelDepthOffsetMode=PDOM_Legacy
  usedShadingModels=Default Lit
 }
 nodes [
  ActorPositionWS()
  n1 = DynamicParameter() {paramNames [Wobble,Random,WobbleSPeed,Param4]}
  n2 = Constant3Vector() {constant=<rgba 0 1 0 0>}
  n3 = Constant3Vector() {constant=<rgba 0 0 1 0>}
  n4 = Multiply(A=Time(), B=@n1.WobbleSPeed)
  n5 = Constant3Vector() {constant=<rgba 15 0 0 0>}
  n6 = RotateAboutAxis(
   NormalizedRotationAxis=Transform(None=LinearInterpolate(A=@n2, B=@n3, Alpha=@n1.Random)) {transformSourceType=TRANSFORMSOURCE_World, transformType=TRANSFORM_ParticleWorld}
   RotationAngle=Multiply(A=Sine(None=@n4) {period=0.2}) {constB=0.07}
   PivotPoint=Add(A=ObjectPositionWS(), B=Transform(None=@n5) {transformSourceType=TRANSFORMSOURCE_World, transformType=TRANSFORM_ParticleWorld})
   Position=WorldPosition().XYZ
  )
  MaterialFunctionCall!(In=MaterialFunctionCall!(
    Material=MaterialFunctionCall!(UV 1 coordinates, MuzzleFalloff) {materialFunction=/Game/Packages/Fortnite_UniqueMaterials_1/Weapons/M_GunMasterShaderFunction}.Material Result
    WorldPositionOffset=Multiply(A=@n1.Wobble, B=@n6)
   ) {materialFunction=/Engine/Functions/MaterialLayerFunctions/MatLayerBlend_AddWorldPositionOffset}.Blended Material) {materialFunction=/Game/Effects/Niagara/MasterMaterials/Functions/MF_FX_EmissiveAdjustments_MA}
  Multiply(B=@n6)
 ]
}
