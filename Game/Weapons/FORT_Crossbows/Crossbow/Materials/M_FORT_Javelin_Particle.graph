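# Default-lit mesh-particle material for the crossbow Javelin, using crossbow textures and adding a vertex wobble.
# A time-driven sine, scaled by WobbleDistance, rotates vertices about an offset pivot on a particle-space axis blended between Y and Z.
# Dynamic parameters supply wobble strength, speed and the random axis blend per particle.
# The second UV set's V coordinate weights the motion along the mesh.
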Material /Game/Weapons/FORT_Crossbows/Crossbow/Materials/M_FORT_Javelin_Particle
graph {
 properties {bUsedWithMeshParticles=true, bUsedWithNiagaraMeshParticles=true, pixelDepthOffsetMode=PDOM_Legacy, usedShadingModels=Default Lit}
 nodes [
  n1 = ComponentMask(None=TextureCoordinate() {coordinateIndex=1}) {g=true}
  n2 = DynamicParameter() {paramNames [Wobble,Random,WobbleSPeed,Param4]}
  n3 = Transform(None=LinearInterpolate(A=Constant3Vector() {constant=<rgba 0 1 0 0>}, B=Constant3Vector() {constant=<rgba 0 0 1 0>}, Alpha=@n2.Random)) {transformSourceType=TRANSFORMSOURCE_World, transformType=TRANSFORM_ParticleWorld}
  n4 = Time()
  n5 = Multiply(A=Sine(None=Multiply(A=@n4, B=@n2.WobbleSPeed)) {period=0.2}, B=ScalarParameter() {defaultValue=0.07, parameterName=WobbleDistance}) {constB=0.07}
  n6 = Add(
   A=ObjectPositionWS()
   B=Transform(None=VectorParameter() {parameterName=PivotPoint-Offset, defaultValue=<rgba 15 0 0 1>}.RGB) {transformSourceType=TRANSFORMSOURCE_World, transformType=TRANSFORM_ParticleWorld}
  )
  n7 = WorldPosition()
  Multiply(
   A=@n2.Wobble
   B=Multiply(
    A=LinearInterpolate(Alpha=@n1) {constA=0.2}
    B=RotateAboutAxis(NormalizedRotationAxis=@n3, RotationAngle=@n5, PivotPoint=@n6, Position=@n7.XYZ)
   )
  )
  Reroute(None=TextureSampleParameter2D() {parameterName=BaseColor, texture=/Game/Weapons/FORT_Crossbows/Crossbow/Textures/T_Crossbow_01_D}.RGB)
  TextureSampleParameter2D() {parameterName=Normal, samplerType=SAMPLERTYPE_Normal, texture=/Game/Weapons/FORT_Crossbows/Crossbow/Textures/T_Crossbow_01_N}
  TextureSampleParameter2D() {parameterName=Mask, samplerType=SAMPLERTYPE_Masks, texture=/Game/Weapons/FORT_Crossbows/Crossbow/Textures/T_Crossbow_01_S}
 ]
}
