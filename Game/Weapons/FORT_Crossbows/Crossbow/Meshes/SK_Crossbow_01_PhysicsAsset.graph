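# Physics asset for the SK_Crossbow_01 crossbow skeletal mesh, giving it a simplified collision and physics body setup.
# Two box bodies cover the Crossbow_Handle and LimbBrace bones, each offset and rotated to fit its part of the weapon.
# A single constraint joins LimbBrace to Crossbow_Handle, with both swing limit angles and the twist limit angle set equally.
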PhysicsAsset /Game/Weapons/FORT_Crossbows/Crossbow/Meshes/SK_Crossbow_01_PhysicsAsset
bodies {
 Crossbow_Handle [{shapeType=Box, extentX=16.9497, extentY=36.9197, extentZ=91.9369, center=<20.1993 0.0013 2.321>, rotation=<rot 0 -90.0845 -83.8879>}]
 LimbBrace [{shapeType=Box, extentX=29.7546, extentY=9.4271, extentZ=77.48, center=<-2.5226 -0.0301 1.3232>, rotation=<rot 0 -180 -90>}]
}
constraints [{bone1Name=LimbBrace, bone2Name=Crossbow_Handle, swing1LimitDegrees=45, swing2LimitDegrees=45, twistLimitDegrees=45}]
